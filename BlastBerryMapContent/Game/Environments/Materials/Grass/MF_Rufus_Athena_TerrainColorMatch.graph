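# Material function that tints grass to match the Athena terrain color beneath it.
# It samples a combined terrain color texture at world-position landscape coordinates passed through a vertex interpolator, with a mip bias for a softened lookup.
# The sample is scaled by ColormapDivisor from the landscape parameter collection, then multiplied by a NeutralGrassColor parameter.
# Outputs provide the tinted result, the neutral grass color alone and the untinted terrain color.
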MaterialFunction /BlastBerryMapContent/Game/Environments/Materials/Grass/MF_Rufus_Athena_TerrainColorMatch
graph {nodes [
  n1 = TextureSampleParameter2D!(
   UVs=MaterialFunctionCall!(WorldPosition) {materialFunction=/BlastBerryMapContent/Game/Athena/Environments/Landscape/MaterialFunctions/Arid/MF_Athena_Terrain_Coords}.WithInterpolator
   Bias
   Apply View MipBias
  ) {
   constMipValue=5
   group=TerrainColorMatch
   mipValueMode=TMVM_MipBias
   parameterName=GrassTextureToMatch
   samplerType=SAMPLERTYPE_LinearColor
   texture=/BlastBerryMapContent/Game/Environments/Materials/Grass/Textures/T_Athena_Terrain_CombinedColors_01
  }
  n2 = Multiply(
   A=@n1.RGB
   B=CollectionParameter() {collection=/BlastBerryMapContent/Game/Athena/Environments/Landscape/MPC/MPC_Landscape, parameterName=ColormapDivisor}
  )
  n3 = VectorParameter() {parameterName=NeutralGrassColor, defaultValue=<rgba 0.0675 0.145 0.02 1>}
  FunctionOutput(None=Multiply(A=@n2, B=@n3.RGB))
  FunctionOutput(None=@n3.RGB) {outputName=GrassColor}
  FunctionOutput(None=@n2) {outputName=GrassNoColor}
 ]}
