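# MF_StormMask is a material function in the grass materials folder that outputs a mask marking areas outside the Athena safe zone, in the storm.
# It reads the safe zone center and radius from the FortniteMaterialParameters collection.
# A hard-edged SphereMask compares world-position XY to the center XY, ignoring height, and is inverted so the output is 1 outside the circle and 0 inside.
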MaterialFunction /BlastBerryMapContent/Game/Environments/Materials/Grass/MF_StormMask
graph {nodes [
  n1 = ComponentMask(None=WorldPosition().XYZ) {g=true, r=true}
  n2 = CollectionParameter() {collection=&FortniteMaterialParameters, parameterName=AthenaSafeZoneCenter}
  n3 = ComponentMask(None=@n2) {g=true, r=true}
  n4 = CollectionParameter() {collection=&FortniteMaterialParameters, parameterName=AthenaSafeZoneRadius}
  n5 = Constant() {r=1}
  FunctionOutput(None=OneMinus(None=SphereMask(A=@n1, B=@n3, Radius=@n4, Hardness=@n5)))
 ]}
