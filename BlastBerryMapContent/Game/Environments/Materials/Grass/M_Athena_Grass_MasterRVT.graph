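# Master grass material for BlastBerry terrain foliage: masked, two-sided subsurface, usable with instancing and Nanite.
# World position offset combines storm-aware panning-noise wind, player bending, physics-field pushes, and sinking by instance fade and camera proximity.
# Gradient-masked root-to-tip tints and distance-faded subsurface color feed the base attributes, with optional Runtime Virtual Texture blending into the landscape.
# Mobile gets biome-aware mountain tinting and a simple height offset instead of animation.
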Material /BlastBerryMapContent/Game/Environments/Materials/Grass/M_Athena_Grass_MasterRVT
graph {
 properties {
  bTangentSpaceNormal=false
  bUseMaterialAttributes=true
  bUsedWithInstancedStaticMeshes=true
  bUsedWithNanite=true
  blendMode=BLEND_Masked
  pixelDepthOffsetMode=PDOM_Legacy
  shadingModel=MSM_Subsurface
  twoSided=true
  usedShadingModels=Subsurface
 }
 nodes [
  n1 = TextureCoordinate() {coordinateIndex=2}
  n2 = MaterialFunctionCall(X=ComponentMask(None=@n1) {r=true}, Y=ComponentMask(None=@n1) {g=true}, Z=Constant()) {materialFunction=&MakeFloat3}
  Add(A=@n2.Result)
  n3 = Subtract(A=WorldPosition().XYZ, B=CameraPositionWS())
  n4 = Constant3Vector() {constant=<rgba 1 1 0 0>}
  Add(
   A=Multiply(A=Multiply(A=@n3, B=@n4), B=ScalarParameter() {defaultValue=1, parameterName=CameraFadeMaskPush})
   B=AppendVector(A=Constant2Vector(), B=ScalarParameter() {defaultValue=-200, parameterName=CameraFadeMaskSink})
  )
  AppendVector(A=Constant2Vector() {g=1, r=1}, B=TwoSidedSign())
  Constant() {r=0.84}
  Constant()
  Constant() {r=-50}
  n5 = MaterialFunctionCall() {materialFunction=&/MF_Rufus_Athena_TerrainColorMatch}
  Desaturation(None=@n5.Result, Fraction=ScalarParameter() {parameterName=testingGrass})
  n6 = NamedRerouteUsage()
  n7 = MaterialFunctionCall() {materialFunction=/BlastBerryMapContent/Environments/Landscape/Materials/MaterialFunctions/MF_BB_BiomeMask}
  LinearInterpolate(
   A=MaterialFunctionCall!(Rock_Diffuse, ColorBlend_Alpha, ColorBlend_Alpha_Bias, Rock_Diffuse1, Grass_Diffuse=@n6) {materialFunction=/BlastBerryMapContent/Environments/Landscape/Materials/MaterialFunctions/MF_Mountain_Colors}.Grass
   B=@n6
   Alpha=@n7.Mountain
  )
  n8 = ComponentMask(None=WorldPosition().XYZ) {g=true, r=true}
  n9 = TextureSample(UVs=Divide(A=@n8, B=ScalarParameter() {defaultValue=1000, group=WPO, parameterName=HeightVarianceScale})) {samplerType=SAMPLERTYPE_LinearColor, texture=/Game/Environments/Foliage/Textures/T_FieldGrass_01_M}
  n10 = LinearInterpolate(
   A=ScalarParameter() {group=WPO, parameterName=StaticWPOLow}
   B=ScalarParameter() {defaultValue=30, group=WPO, parameterName=StaticWPOHigh}
   Alpha=@n9.R
  )
  n11 = PerInstanceFadeAmount()
  n12 = Multiply(
   A=Multiply(
    A=@n9.R
    B=CollectionParameter() {collection=/BlastBerryMapContent/Game/Athena/Environments/Landscape/MPC/MPC_Landscape, parameterName=Ambient_WindMotion_Amount}
   )
   B=Constant() {r=300}
  )
  n13 = Divide(A=@n8, B=ScalarParameter() {defaultValue=1200, group=WPO, parameterName=WindWaveScale})
  n14 = MaterialFunctionCall() {materialFunction=&/MF_StormMask}
  n15 = Multiply(
   A=Time()
   B=LinearInterpolate(
    A=CollectionParameter() {collection=&MPC_Landscape, parameterName=Grass_WindSpeed}
    B=CollectionParameter() {collection=&MPC_Landscape, parameterName=Grass_WindSpeed_Storm}
    Alpha=@n14.Result
   )
  )
  n16 = Multiply(
   A=TextureSample(UVs=Panner(Coordinate=@n13, Time=@n15) {speedY=1}) {samplerType=SAMPLERTYPE_Normal, texture=/Game/Global/Textures/Default/Noise/T_SwirlNoise_N}.RGB
   B=LinearInterpolate(
    A=CollectionParameter() {collection=&MPC_Landscape, parameterName=Grass_WindPower}
    B=CollectionParameter() {collection=&MPC_Landscape, parameterName=Grass_WindPower_Storm}
    Alpha=@n14.Result
   )
  )
  n17 = Normalize(VectorInput=AppendVector(A=ComponentMask(None=@n16) {g=true, r=true}, B=Constant() {r=1}))
  n18 = Add(
   A=Multiply(A=Constant3Vector() {constant=<rgba 0 0 1 0>}, B=Multiply(A=@n10, B=@n11))
   B=Multiply(A=Multiply(A=@n12, B=@n17), B=Constant3Vector() {constant=<rgba 1 1 0 0>})
  )
  n19 = NamedRerouteUsage()
  n20 = ScalarParameter() {defaultValue=1, group=Animation, parameterName=MaxRotation_PhysicsField}
  LinearInterpolate(A=@n18, B=Multiply(A=Multiply(A=@n19, B=@n20), B=NamedRerouteUsage()), Alpha=NamedRerouteUsage())
  n21 = SetMaterialAttributes!(
   MaterialAttributes
   Base Color=Multiply(A=VectorParameter() {parameterName=wefq34rtqwt, defaultValue=<rgba 1 1 1 1>}.RGB, B=NamedRerouteUsage())
   Specular=LinearInterpolate(
    A=ScalarParameter() {defaultValue=0.02, parameterName=SpecMin}
    B=ScalarParameter() {defaultValue=0.2, parameterName=Spec}
    Alpha=NamedRerouteUsage()
   )
   Roughness=ShadingPathSwitch(Default=ScalarParameter() {defaultValue=0.6, parameterName=Rough}, Mobile=Constant() {r=1})
   Opacity=NamedRerouteUsage()
   Opacity Mask=TextureSampleParameter2D() {parameterName=Grass_Mask, texture=/Game/Environments/Foliage/Textures/T_WavyGrass_01}.R
   Normal=NamedRerouteUsage()
   World Position Offset=NamedRerouteUsage()
   Subsurface Color=NamedRerouteUsage()
  ) {attributeSetTypes [
    #1
    #2
    #3
    #4
    #5
    #6
    #7
    #8
   ]}
  MaterialFunctionCall!(In=NamedRerouteDeclaration(None=@n21) {name=Pre RVT, nodeColor=<rgba 1 0.7734 0 1>}) {materialFunction=/BlastBerryMapContent/Game/TechArt/Material_Functions/MF_Rufus_RVTBlendingOn}
  n22 = Saturate(None=LinearInterpolate(
    A=VectorParameter() {parameterName=Grass_DefaultMobileTint, defaultValue=<rgba 0.0419 0.09 0.0124 1>}.RGB
    B=VectorParameter() {parameterName=Grass_DefaultMobileTipTint, defaultValue=<rgba 0.0838 0.18 0.0248 1>}.RGB
    Alpha=NamedRerouteUsage()
   ))
  Multiply(
   A=@n22
   B=LinearInterpolate(
    A=ScalarParameter() {defaultValue=0.8, parameterName=PerInstanceRandomLow_Mobile}
    B=ScalarParameter() {defaultValue=1.25, parameterName=PerInstanceRandomHigh_Mobile}
    Alpha=PerInstanceRandom()
   )
  )
  Multiply(A=@n5.GrassNoColor, B=VectorParameter() {parameterName=Arid_SandGrass_Tint, defaultValue=<rgba 0.4678 0.314 0.1384 1>}.RGB)
  n23 = LinearInterpolate(
   A=VectorParameter() {parameterName=Grass_DefaultTint, defaultValue=<rgba 0.0675 0.145 0.02 1>}.RGB
   B=VectorParameter() {parameterName=Grass_DefaultTipTint, defaultValue=<rgba 0.0838 0.18 0.0248 1>}.RGB
   Alpha=NamedRerouteUsage()
  )
  n24 = MaterialFunctionCall(Grass_Diffuse=@n22) {materialFunction=&MF_Mountain_Colors}
  n25 = MaterialFunctionCall() {materialFunction=&MF_BB_BiomeMask}
  NamedRerouteDeclaration(None=ShadingPathSwitch(Default=Saturate(None=@n23), Mobile=LinearInterpolate(A=@n22, B=@n24.Grass, Alpha=@n25.Mountain))) {name=Base_Color, nodeColor=<rgba 0 0.7422 1 1>}
  n26 = NamedRerouteUsage()
  n27 = ScalarParameter() {defaultValue=0.1, desc=Multi to control sss color strength, group=RufusSubSurface, parameterName=SSS_Multi}
  n28 = Divide(
   A=Add(A=PixelDepth().R, B=ScalarParameter() {defaultValue=-7000, group=Camera Fade Distances, parameterName=FarFadeOffset})
   B=ScalarParameter() {defaultValue=4000, group=Camera Fade Distances, parameterName=FarFade}
  )
  NamedRerouteDeclaration(None=LinearInterpolate(A=Multiply(A=@n26, B=@n27), Alpha=Saturate(None=@n28)) {constB=0}) {name=SSS Color, nodeColor=<rgba 0.5703 1 0 1>}
  n29 = TextureCoordinate() {coordinateIndex=1}
  n30 = NamedRerouteUsage()
  n31 = ScalarParameter() {defaultValue=1, group=Animation, parameterName=MaxRotation_PhysicsField}
  n32 = Multiply(
   A=Multiply(A=Multiply(A=@n30, B=@n31), B=NamedRerouteUsage())
   B=ScalarParameter() {defaultValue=86.75, parameterName=Explosion Strenth modifier}
  ) {constB=86.75}
  n33 = NamedRerouteUsage()
  n34 = CollectionParameter() {collection=&FortniteMaterialParameters, parameterName=PlayerPosition}
  n35 = MaterialFunctionCall!(
   PlayerPosition=@n34
   PlayerRadiusMax=CollectionParameter() {collection=&MPC_Landscape, parameterName=PlayerBend_Radius}
   BendStrength=CollectionParameter() {collection=&MPC_Landscape, parameterName=PlayerBend_Strength}
  ) {materialFunction=/Game/Athena/Environments/Landscape/MaterialFunctions/MF_AthenaFoliagePlayerInteract}
  n36 = Multiply(
   A=OneMinus(None=ComponentMask(None=@n29) {g=true})
   B=LinearInterpolate(A=LinearInterpolate(A=@n18, B=@n32, Alpha=@n33), B=@n35.Rotation, Alpha=@n35.Rotation)
  )
  n37 = Constant2Vector()
  n38 = Multiply(A=OneMinus(None=PerInstanceFadeAmount()), B=ScalarParameter() {defaultValue=-65, parameterName=SinkIntoGroundDist})
  n39 = Multiply(
   A=SphereMask(
    A=CameraPositionWS()
    B=WorldPosition().XYZ
    Radius=ScalarParameter() {defaultValue=200, parameterName=camerafadeMaskRadius}
    Hardness=ScalarParameter() {parameterName=camerafadeMaskHardness}
   )
   B=Constant() {r=-200}
  )
  n40 = Constant()
  n41 = CollectionParameter() {collection=&FortniteMaterialParameters, parameterName=IsPlayerSliding}
  n42 = Add(
   A=Add(A=QualitySwitch(Default=@n36, High=@n36, Medium=@n36), B=AppendVector(A=@n37, B=@n38))
   B=AppendVector(A=Constant2Vector(), B=LinearInterpolate(A=@n39, B=@n40, Alpha=@n41))
  )
  n43 = ComponentMask(None=TextureCoordinate() {coordinateIndex=1}) {g=true}
  n44 = Multiply(
   A=Multiply(A=OneMinus(None=@n43), B=Constant3Vector() {constant=<rgba 0 0 1 0>})
   B=ScalarParameter() {defaultValue=15, group=WPO, parameterName=MobileGrassHeightOffset}
  )
  n45 = LinearInterpolate(A=ScalarParameter() {defaultValue=-50, group=WPO, parameterName=SinkIntoGroundAmount}, B=Constant(), Alpha=PerInstanceFadeAmount())
  n46 = Constant3Vector() {constant=<rgba 0 0 1 0>}
  NamedRerouteDeclaration(None=Add(A=ShadingPathSwitch(Default=@n42, Mobile=@n44), B=Multiply(A=@n45, B=@n46)) {constB=0}) {name=WPO, nodeColor=<rgba 0 0.0174 1 1>}
  n47 = VertexInterpolator(VS=Transform(None=Constant3Vector() {desc=Looks almost same RVT normal, constant=<rgba 0 0 1 0>}) {transformSourceType=TRANSFORMSOURCE_Instance, transformType=TRANSFORM_Tangent})
  n48 = MaterialFunctionCall(
   Normal=TextureSample() {samplerType=SAMPLERTYPE_Normal, texture=/Game/Environments/Foliage/Textures/T_WavyGrass_01_N}.RGB
   Flatness=ScalarParameter() {defaultValue=0.84, group=Normal Settings, parameterName=TextureNormalFlatten}
  ) {materialFunction=&FlattenNormal}
  NamedRerouteDeclaration(None=Transform(None=MaterialFunctionCall(BaseNormal=@n47.PS, AdditionalNormal=@n48.Result) {materialFunction=&BlendAngleCorrectedNormals}.Result)) {name=normal map, nodeColor=<rgba 0.9219 0 1 1>}
  n49 = ScalarParameter() {defaultValue=1, group=SubSurface Settings, parameterName=SSSOpacity}
  n50 = ScalarParameter() {defaultValue=0.5, group=SubSurface Settings, parameterName=SSSOpacity2}
  n51 = MaterialFunctionCall() {materialFunction=&LinearGradient}
  n52 = Subtract(A=@n51.VGradient, B=ScalarParameter() {defaultValue=0.624, group=SubSurface Settings, parameterName=SSS gradient adjust})
  NamedRerouteDeclaration(None=Saturate(None=LinearInterpolate(A=@n49, B=@n50, Alpha=@n52))) {name=SSSOpacity, nodeColor=<rgba 0.1719 0 1 1>}
  n53 = SamplePhysicsVectorField!(World Position=NamedRerouteDeclaration(None=TransformPosition(None=@n2.Result) {transformSourceType=TRANSFORMPOSSOURCE_Instance, transformType=TRANSFORMPOSSOURCE_World}) {name=PivotPosition, nodeColor=<rgba 0 0.9766 1 1>}) {fieldTarget=Vector_LinearVelocity}
  n54 = MaterialFunctionCall(Vector 3=Add(A=@n53) {constB=0.0001}) {materialFunction=/Engine/Functions/Engine_MaterialFunctions02/Utility/VectorLength}
  n55 = Constant()
  NamedRerouteDeclaration(None=ShadingPathSwitch(Default=Saturate(None=@n54.V3 Length), Mobile=@n55) {desc=Migrated from FeatureLevelSwitch}) {name=PhysicsFieldMask, nodeColor=<rgba 0 1 0.3906 1>}
  n56 = Constant3Vector() {constant=<rgba 0 0 1 0>}
  NamedRerouteDeclaration(None=ShadingPathSwitch(
    Default=MaterialFunctionCall(Vector=@n53, Default=@n56) {materialFunction=&SafeNormalize}.Result
    Mobile=Constant3Vector() {constant=<rgba 0 0 1 0>}
   ) {desc=Migrated from FeatureLevelSwitch}) {name=PhysicsFieldDirection, nodeColor=<rgba 0.3828 0 1 1>}
  n57 = Min(A=@n54.V3 Length, B=ScalarParameter() {defaultValue=1, group=Animation, parameterName=MaxRotation_PhysicsField_Clamp})
  n58 = Time()
  n59 = ScalarParameter() {defaultValue=351, group=Animation, parameterName=PhysicsField_NoiseOffset}
  n60 = Frac(None=Multiply(A=ScalarParameter() {defaultValue=2.5, group=Animation, parameterName=MaxRotation_PhysicsField_NoiseSpeed}, B=Add(A=@n58, B=@n59)))
  n61 = OneMinus(None=@n60)
  n62 = LinearInterpolate(
   A=Multiply(A=Multiply(A=@n60, B=@n61)) {constB=2}
   Alpha=ScalarParameter() {defaultValue=0.6, group=Animation, parameterName=MaxRotation_PhysicsField_MInBend}
  )
  NamedRerouteDeclaration(None=ShadingPathSwitch(Default=Multiply(A=@n57, B=@n62), Mobile=@n55) {desc=Migrated from FeatureLevelSwitch}) {name=PhysicsFieldStrength, nodeColor=<rgba 1 0 0.2109 1>}
  NamedRerouteDeclaration(None=@n35.PlayerSphereMask) {name=Player SPhere mask, nodeColor=<rgba 0.9688 0 1 1>}
  NamedRerouteDeclaration(None=@n35.Rotation) {name=PLayer Dir, nodeColor=<rgba 0 1 0.6719 1>}
  n63 = TextureSampleParameter2D() {parameterName=Gradient_Mask, samplerType=SAMPLERTYPE_LinearColor, texture=/Game/Athena/Prototype/Terrain/T_GrassGradient_01}
  NamedRerouteDeclaration(None=Power(Base=@n63.G, Exp=Constant() {r=3})) {name=TipMask, nodeColor=<rgba 0 0.7422 1 1>}
  NamedRerouteDeclaration(None=@n63.G) {name=TipMaskMobile, nodeColor=<rgba 0 0.7422 1 1>}
  NamedRerouteDeclaration(None=Power(Base=@n63.G, Exp=ScalarParameter() {defaultValue=3, parameterName=RVT_BlendHeightPow})) {name=RVT_TipMask, nodeColor=<rgba 0 0.7422 1 1>}
  OneMinus(None=@n7.Mountain)
  Power(Base=@n63.G, Exp=ScalarParameter() {defaultValue=1, parameterName=TipMaskMobilePow})
  Reroute(None=@n51.VGradient)
  Reroute(None=VertexColor().B)
  n64 = CollectionParameter() {collection=&MPC_Landscape, parameterName=Arid_Grass_Tint}
  n65 = MaterialFunctionCall() {materialFunction=&/MF_Athena_BiomeMask}
  n66 = StaticSwitchParameter(
   True=Multiply(A=@n5.GrassColor, B=ComponentMask(None=@n64) {b=true, g=true, r=true})
   False=LinearInterpolate(A=@n5.Result, B=@n5.Result, Alpha=@n65.Swamp)
  ) {parameterName=ForceAridGrass}
  n67 = VectorParameter() {parameterName=GrassSample_Tint, defaultValue=<rgba 0.9 0.9 0.9 0>}
  n68 = VectorParameter() {parameterName=GrassSample_Tint_Arid, defaultValue=<rgba 0.9 0.9 0.9 0>}
  n69 = VectorParameter() {parameterName=GrassTipTint, defaultValue=<rgba 1.1 1.1 1.1 1>}
  n70 = LinearInterpolate(
   A=StaticSwitchParameter(True=@n67.RGB, False=LinearInterpolate(A=@n67.RGB, B=@n68.RGB, Alpha=@n65.Swamp)) {parameterName=ForceAridGrass}
   B=ShadingPathSwitch(Default=@n69.RGB, Mobile=Multiply(A=@n69.RGB) {constB=1.2}) {desc=Migrated from FeatureLevelSwitch}
   Alpha=NamedRerouteUsage()
  )
  Reroute(None=Saturate(None=Multiply(A=@n66, B=@n70)))
  ScalarParameter() {defaultValue=-50, parameterName=PlayerBendStrength}
  ScalarParameter() {defaultValue=200, parameterName=PlayerBendRadius_1}
  ScalarParameter() {defaultValue=0.08, group=WPO, parameterName=WindSpeed}
  ScalarParameter() {defaultValue=0.75, group=WPO, parameterName=WindSpeed_1}
  ScalarParameter() {defaultValue=0.6, group=WPO, parameterName=WindPower}
  ScalarParameter() {defaultValue=1, group=WPO, parameterName=WindPower_Storm}
  ScalarParameter() {defaultValue=0.4, parameterName=Ambient Wind Motion Amount}
  n71 = NamedRerouteUsage()
  n72 = VertexNormalWS()
  n73 = RuntimeVirtualTextureSampleParameter!(UVs, Translated World Position, Mip Bias=LinearInterpolate(Alpha=ComponentMask(None=@n72) {b=true}) {constA=6, constB=0}) {
   group=D_RVT
   materialType=BaseColor_Normal_Specular_YCoCg
   mipValueMode=RVTMVM_MipBias
   parameterName=RVT
   sortPriority=1
   virtualTexture=/BlastBerryMapContent/Game/Athena/Environments/Materials/RVT/Athena_RuntimeVirtualTexture
   worldPositionOriginType=CameraRelative
  }
  n74 = SetMaterialAttributes!(
   MaterialAttributes=@n71
   Base Color=LinearInterpolate(
    A=NamedRerouteDeclaration(None=@n73.BaseColor) {name=RVT Base Color, nodeColor=<rgba 0 1 0.6016 1>}
    B=NamedRerouteUsage()
    Alpha=NamedRerouteUsage()
   )
   Specular=NamedRerouteDeclaration(None=@n73.Specular) {name=RVT Specular, nodeColor=<rgba 1 0 0.8203 1>}
   Roughness=NamedRerouteDeclaration(None=@n73.Roughness) {name=RVT Rough, nodeColor=<rgba 0 1 0.7422 1>}
   Normal=NamedRerouteDeclaration(None=@n73.Normal) {name=RVT N, nodeColor=<rgba 0 0.6953 1 1>}
  ) {attributeSetTypes [
    #1
    #2
    #3
    #6
   ]}
  StaticSwitchParameter(
   True=MaterialFunctionCall(NoRVT=@n71, RVT=@n74) {materialFunction=&FN_RVT_Switch_Material}.Result
   False=NamedRerouteUsage()
  ) {group=FlowerFade, parameterName=UseRVT}
  VectorParameter() {parameterName=Grass_DefaultMobileMountainTint, defaultValue=<rgba 0.0093 0.02 0.0028 1>}
 ]
}
